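# M_HoloMap is an additive, unlit material for a holographic map of animated, abstract terrain.
# Rotating mask and noise textures build a height field that drives world position offset and a stepped-contour terrain look.
# Grid lines, dots, derivative outlines, scrolling static, sweeps, timed switching and edge and distance fades add the hologram effects.
# Platform switches simplify costlier layers, dithering softens opacity, and MF_FN_DestructibleDamage adds damage animation.
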Material /Hermes/Techart/Crewcut/M_HoloMap
graph {
 properties {bAutomaticallySetUsageInEditor=false, bUseMaterialAttributes=true, blendMode=BLEND_Additive, shadingModel=MSM_Unlit, usedShadingModels=Unlit}
 nodes [
  ActorPositionWS()
  n1 = TextureCoordinate()
  n2 = ComponentMask(None=@n1) {r=true}
  Add(A=Multiply(A=@n2, B=ScalarParameter() {defaultValue=1, parameterName=Fade X}) {constB=2}, B=ScalarParameter() {parameterName=Fade Pos X})
  AppendVector()
  n3 = NamedRerouteUsage()
  ComponentMask(None=@n3) {g=true, r=true}
  Constant() {r=40}
  Constant() {r=600}
  Constant() {r=1}
  n4 = MaterialFunctionCall(
   Plane Normal Axis=Constant3Vector() {constant=<rgba 0 0 1 0>}
   Plane X Axis=Constant3Vector() {constant=<rgba 1 0 0 0>}
   Plane Center=ObjectPositionWS()
   UV Scale=Constant() {r=1}
  ) {materialFunction=&VirtualPlaneCoordinates}
  Divide(A=@n4.UVs) {constB=150}
  n5 = LinearInterpolate(A=TextureCoordinate(), B=NamedRerouteUsage()) {constAlpha=0.3}
  n6 = ComponentMask(None=Multiply(A=ScalarParameter() {defaultValue=4, parameterName=Dot Amount}, B=Add(A=@n3, B=@n5)) {constB=8}) {g=true, r=true}
  n7 = Time()
  n8 = TextureSample(UVs=Panner(Coordinate=@n6, Time=Floor(None=@n7)) {speedX=0.2, speedY=0.2}) {texture=&/Dot_Noise}
  n9 = Divide(A=Time()) {constB=1.37}
  n10 = TextureSample(UVs=Panner(Coordinate=Multiply(A=@n6) {constB=1.5}, Time=Floor(None=@n9)) {speedX=-0.2, speedY=-0.2}) {texture=&Dot_Noise}
  MaterialFunctionCall(In=Multiply(A=@n8.G, B=@n10.R), Contrast=Constant() {r=1}) {materialFunction=&CheapContrast}
  MaterialFunctionCall!(X range Y range, Gradient Height, ShowGrid?, Thickness, Color, "f(x)") {materialFunction=/Engine/Functions/Engine_MaterialFunctions01/Debug/PlotFunctionOnGraph}
  n11 = NamedRerouteUsage()
  Multiply(A=Saturate(None=OneMinus(None=@n11))) {constB=0.025}
  n12 = AppendVector(A=NamedRerouteUsage(), B=Constant())
  n13 = Multiply(A=@n12, B=ScalarParameter() {defaultValue=15, parameterName=Line Amount}) {constB=25}
  n14 = MaterialFunctionCall!(
   In=Power(Base=Sine(None=@n13), Exp=ScalarParameter() {defaultValue=87.6978, parameterName=Line Thickness}) {constExponent=20.155}
   Contrast=Constant() {r=4}
  ) {materialFunction=/Engine/Functions/Engine_MaterialFunctions01/ImageAdjustment/CheapContrast_RGB}
  n15 = ComponentMask(None=@n14.Result) {r=true}
  n16 = ComponentMask(None=@n14.Result) {g=true}
  n17 = Add(A=@n15, B=@n16)
  n18 = OneMinus(None=Add(A=ComponentMask(None=@n14.Result) {r=true}, B=ComponentMask(None=@n14.Result) {g=true}))
  n19 = Multiply(A=@n12, B=Multiply(A=ScalarParameter() {defaultValue=15, parameterName=Line Amount}) {constB=5}) {constB=25}
  n20 = ScalarParameter() {defaultValue=87.6978, parameterName=Line Thickness}
  n21 = MaterialFunctionCall(In=Power(Base=Sine(None=@n19), Exp=Multiply(A=@n20) {constB=0.1}) {constExponent=20.155}, Contrast=Constant() {r=4}) {materialFunction=&CheapContrast_RGB}
  n22 = ComponentMask(None=@n21.Result) {r=true}
  n23 = ComponentMask(None=@n21.Result) {g=true}
  n24 = Multiply(A=Add(A=@n22, B=@n23)) {constB=0.7}
  n25 = Multiply(A=Add(A=Multiply(A=@n17) {constB=2.5}, B=Multiply(A=@n18, B=@n24)), B=ScalarParameter() {defaultValue=0.01, parameterName=Line Intensity})
  n26 = Add(A=Multiply(A=ScalarParameter() {defaultValue=0.3, parameterName=Dots Intensity}, B=Multiply(A=@n15, B=@n16)) {constA=30}, B=@n25)
  n27 = Multiply(A=@n22, B=@n23)
  n28 = ScalarParameter() {defaultValue=0.3, parameterName=Dots Intensity}
  n29 = Multiply(A=NamedRerouteUsage()) {constB=1.2}
  n30 = Multiply(A=Add(A=Add(A=@n26, B=@n25), B=Multiply(A=@n27, B=@n28) {constA=30}), B=Saturate(None=OneMinus(None=@n29)))
  Multiply(A=OneMinus(None=NamedRerouteUsage()), B=@n30)
  n31 = ScalarParameter() {parameterName=Fade Pos Y}
  n32 = Multiply(A=ScalarParameter() {defaultValue=2, parameterName=Fade Y}, B=ComponentMask(None=@n1) {g=true}) {constB=2}
  n33 = Abs(None=OneMinus(None=Add(A=@n31, B=@n32)))
  n34 = Abs(None=OneMinus(None=@n2))
  n35 = Power(
   Base=Multiply(A=OneMinus(None=@n33), B=OneMinus(None=@n34))
   Exp=LinearInterpolate(A=Constant() {r=0.9}, B=ScalarParameter() {defaultValue=0.55, parameterName=UV Fade Size}, Alpha=NamedRerouteUsage())
  )
  NamedRerouteDeclaration(None=Saturate(None=Multiply(A=@n35) {constB=2})) {name=Fade, nodeColor=<rgba 0.5 1 0 1>}
  n36 = NamedRerouteUsage()
  n37 = Constant2Vector() {g=0.1, r=5}
  n38 = NamedRerouteUsage()
  n39 = TextureSample(UVs=Panner(Coordinate=@n36, Speed=Multiply(A=@n37, B=@n38)) {speedX=5, speedY=0.1}) {
   samplerSource=SSM_Wrap_WorldGroupSettings
   samplerType=SAMPLERTYPE_Masks
   texture=/Game/Missions/Primary/RepairTheShelter/Textures/Cinematic_Screen_Static
  }
  n40 = Constant2Vector() {g=-0.05, r=6}
  n41 = NamedRerouteUsage()
  n42 = TextureSample(UVs=Panner(Coordinate=@n36, Speed=Multiply(A=@n40, B=@n41)) {speedX=6, speedY=-0.05}) {
   samplerSource=SSM_Wrap_WorldGroupSettings
   samplerType=SAMPLERTYPE_Masks
   texture=&Cinematic_Screen_Static
  }
  n43 = Constant2Vector() {g=-0.03, r=-3}
  n44 = NamedRerouteUsage()
  n45 = TextureSample(UVs=Panner(Coordinate=@n36, Speed=Multiply(A=@n43, B=@n44)) {speedX=-3, speedY=-0.03}) {
   samplerSource=SSM_Wrap_WorldGroupSettings
   samplerType=SAMPLERTYPE_Masks
   texture=&Cinematic_Screen_Static
  }
  n46 = Multiply(A=@n42.G, B=@n45.R)
  n47 = Add(A=Multiply(A=@n39.B) {constB=0.1}, B=@n46)
  n48 = Add(A=Multiply(A=@n45.R, B=@n45.R), B=@n45.B)
  NamedRerouteDeclaration(None=Multiply(
    A=MaterialFunctionCall("Xb1,PS4"=@n47, "PC-High&&Epic,XSX,PS5"=@n47, "PC-Perf,Android"=@n48, Switch=@n48, "PC-Low(NotPerf)&&Medium"=@n48) {materialFunction=&PlatformSwitch_1D}.Result
    B=ScalarParameter() {defaultValue=46, group=MonitorStatic, parameterName=Static Scale}
   )) {name=Static, nodeColor=<rgba 0 1 0.8828 1>}
  n49 = AppendVector(A=@n46, B=Multiply(A=@n45.G, B=@n39.R))
  n50 = AppendVector(A=@n45.R, B=@n45.R)
  NamedRerouteDeclaration(None=Multiply(
    A=MaterialFunctionCall!("Xb1,PS4"=@n49, "PC-High&&Epic,XSX,PS5"=@n49, "PC-Perf,Android"=@n50, Switch=@n50, "PC-Low(NotPerf)&&Medium"=@n50) {materialFunction=/Game/Effects/Niagara/MasterMaterials/Functions/Common/PlatformSwitch_2D}.Result
    B=ScalarParameter() {defaultValue=0.032, group=MonitorStatic, parameterName=Static UV Distort}
   )) {name=StaticUVDistrotion, nodeColor=<rgba 1 0.5625 0 1>}
  n51 = TextureCoordinate()
  n52 = ComponentMask(None=VectorParameter() {group=MonitorStatic, parameterName=TexCoordScale, defaultValue=<rgba 0.5 0.5 0 0>}.RGB) {g=true, r=true}
  n53 = VectorParameter() {group=MonitorStatic, parameterName=TexCoordOffset}
  NamedRerouteDeclaration(None=Add(A=Multiply(A=@n51, B=@n52), B=ComponentMask(None=@n53.RGB) {g=true, r=true})) {name=TexCoord, nodeColor=<rgba 1 0 0.4453 1>}
  n54 = Multiply(A=ComponentMask(None=NamedRerouteUsage()) {r=true}) {constB=0.75}
  n55 = Multiply(A=Time()) {constB=-0.2}
  n56 = Power(Base=Frac(None=Add(A=@n54, B=@n55))) {constExponent=20}
  NamedRerouteDeclaration(None=Multiply(A=@n56, B=ScalarParameter() {defaultValue=0.3, group=MonitorStatic, parameterName=Sweep Add}) {constB=0.2}) {name=Sweep, nodeColor=<rgba 0.2422 0 1 1>}
  n57 = ScalarParameter() {defaultValue=0.0015, group=MonitorStatic, parameterName=Sweep UV Distort}
  NamedRerouteDeclaration(None=AppendVector(A=Multiply(A=@n56, B=@n57) {constB=0.02}, B=Constant())) {name=SweepUVDistortion, nodeColor=<rgba 0.0781 1 0 1>}
  n58 = TextureCoordinate()
  n59 = AppendVector(A=Add(B=ComponentMask(None=@n58) {g=true}) {constA=-0.5}, B=Add(A=ComponentMask(None=@n58) {r=true}) {constB=-1})
  n60 = Multiply(A=@n59, B=ScalarParameter() {defaultValue=0.12, parameterName=Uv Size})
  n61 = Time()
  n62 = NamedRerouteDeclaration(None=Add(A=NamedRerouteUsage(), B=Rotator(Coordinate=@n60, Time=@n61) {centerX=0, centerY=0, speed=0.075})) {name=UVs, nodeColor=<rgba 1 0 0.1875 1>}
  n63 = Constant() {r=1}
  n64 = TextureSample!(UVs=Add(A=@n62) {constB=0.37}, Tex, Level=@n63, Apply View MipBias) {mipValueMode=TMVM_MipLevel, texture=/Game/Global/Textures/Default/Masks_and_Gradients/chunkma_2}
  n65 = Multiply(A=@n62) {constB=2}
  n66 = Multiply(
   A=Multiply(A=TextureSample(UVs=@n65, Level=@n63) {mipValueMode=TMVM_MipLevel, texture=&chunkma_2}.R) {constB=2}
   B=@n64.G
  )
  n67 = Multiply(A=@n64.G, B=@n66)
  n68 = Add(A=Multiply(A=@n62) {constB=9}, B=NamedRerouteUsage())
  n69 = Constant() {r=2}
  n70 = Constant() {r=0.5}
  n71 = Add(A=NamedRerouteUsage()) {constB=0.7}
  n72 = TextureSample(UVs=Multiply(A=Add(A=@n62, B=@n71) {constB=0.7}) {constB=8}, Level=@n69) {mipValueMode=TMVM_MipLevel, texture=&/T_Holotable_HeightNoise}
  n73 = Multiply(A=@n72.R, B=@n72.R)
  n74 = LinearInterpolate(
   A=Multiply(
    A=TextureSample(UVs=@n68, Level=@n69) {mipValueMode=TMVM_MipLevel, texture=&T_Holotable_HeightNoise}.R
    B=Add(A=@n67, B=@n66)
   )
   B=Multiply(A=Add(A=@n64.R, B=@n70) {constB=0.8}, B=Multiply(A=@n73) {constB=3})
   Alpha=NamedRerouteUsage()
  )
  n75 = Saturate(None=@n74)
  NamedRerouteDeclaration(None=@n75) {name=Height Noise, nodeColor=<rgba 1 0.9375 0 1>}
  n76 = ScalarParameter() {defaultValue=650, parameterName=Distance}
  n77 = ScalarParameter() {defaultValue=223.7961, parameterName=Distance Perf}
  n78 = Divide(
   A=PixelDepth().R
   B=MaterialFunctionCall!("Xb1,PS4"=@n76, "PC-High&&Epic,XSX,PS5"=@n76, "PC-Perf,Android"=@n77, Switch=@n77, "PC-Low(NotPerf)&&Medium"=@n77) {materialFunction=/Game/Effects/Niagara/MasterMaterials/Functions/Common/PlatformSwitch_3D}.Result
  )
  NamedRerouteDeclaration(None=LinearInterpolate(Alpha=Saturate(None=@n78)) {constA=1, constB=0}) {name=Distance, nodeColor=<rgba 1 0 0.4687 1>}
  n79 = Constant() {r=0.05}
  n80 = Power(Base=@n75, Exp=ScalarParameter() {defaultValue=0.2594, parameterName=WPO Power})
  n81 = Multiply(
   A=Multiply(A=Add(A=@n79, B=@n80), B=ScalarParameter() {defaultValue=17, parameterName=WPO Intensity}) {constB=8}
   B=Constant3Vector() {constant=<rgba 0 0 1 0>}
  )
  n82 = Add(A=Multiply(A=NamedRerouteUsage(), B=Constant3Vector() {constant=<rgba 7 7 0 0>}) {constB=9}, B=@n81)
  n83 = NamedRerouteUsage()
  n84 = NamedRerouteUsage()
  NamedRerouteDeclaration(None=Multiply(A=LinearInterpolate(A=@n82, B=@n81, Alpha=@n83), B=Add(A=@n84) {constB=0.4})) {name=WPO, nodeColor=<rgba 1 0 0.3984 1>}
  n85 = TextureCoordinate()
  n86 = AppendVector(A=@n85, B=@n85)
  n87 = Constant() {desc=Colums, r=64}
  n88 = Constant() {desc=Rows, r=1}
  n89 = AppendVector(A=@n87, B=@n88)
  n90 = AppendVector(A=@n89, B=@n89)
  n91 = AppendVector(A=Multiply(A=@n87, B=@n88), B=@n88)
  n92 = Divide(A=Time(), B=Constant() {r=1200}) {constB=500}
  n93 = Frac(None=Add(A=Constant() {r=0.2}, B=@n92)) {desc=0-1 animation}
  n94 = Frac(None=Add(A=@n92, B=Constant() {r=0.8})) {desc=0-1 animation}
  n95 = Multiply(A=AppendVector(A=@n91, B=@n91), B=AppendVector(A=AppendVector(A=@n93, B=@n93), B=AppendVector(A=@n94, B=@n94)))
  n96 = Add(A=Divide(A=Multiply(A=@n86) {constB=0.25}, B=@n90), B=Divide(A=Floor(None=@n95), B=@n90))
  NamedRerouteDeclaration(None=TextureSample!(UVs=ComponentMask(None=@n96) {g=true, r=true}, Tex, Apply View MipBias) {samplerType=SAMPLERTYPE_LinearColor, texture=&/T_RandoMPixelNoise}.R) {name=Random, nodeColor=<rgba 1 0.0703 0 1>}
  n97 = Add(A=Multiply(A=Add(A=@n39.RGB, B=@n42.RGB), B=@n45.RGB), B=@n45.RGB)
  NamedRerouteDeclaration(None=MaterialFunctionCall("Xb1,PS4"=@n97, "PC-High&&Epic,XSX,PS5"=@n97, "PC-Perf,Android"=@n45.RGB, Switch=@n45.RGB, "PC-Low(NotPerf)&&Medium"=@n45.RGB) {materialFunction=&PlatformSwitch_3D}.Result) {name=Static Color, nodeColor=<rgba 0.3359 1 0 1>}
  n98 = ScalarParameter() {defaultValue=0.1084, parameterName=Step Min}
  n99 = ScalarParameter() {defaultValue=1.7912, parameterName=Step Max}
  n100 = Multiply(A=@n74) {constB=6}
  n101 = Constant() {r=0.1}
  n102 = Add(A=Multiply(A=Step(Y=@n100) {constX=0.001}, B=@n101) {constB=0.2}, B=Multiply(A=Step(Y=@n100) {constX=0.005}, B=@n101) {constB=0.2})
  n103 = Multiply(A=Step(Y=@n100) {constX=0.01}, B=@n101) {constB=0.2}
  n104 = Step(Y=@n100) {constX=0.04}
  n105 = Add(A=Add(A=Add(A=@n102, B=@n103), B=Multiply(A=@n104, B=@n101) {constB=0.2}), B=Multiply(A=Step(Y=@n100) {constX=0.1}, B=@n101) {constB=0.2})
  n106 = Multiply(A=Step(Y=@n100) {constX=0.2}, B=@n101) {constB=0.2}
  n107 = Step(Y=@n100) {constX=0.5}
  n108 = Add(A=Add(A=Add(A=@n105, B=@n106), B=Multiply(A=@n107, B=@n101) {constB=0.2}), B=Multiply(A=Step(Y=@n100), B=@n101) {constB=0.2})
  n109 = Multiply(A=Step(Y=@n100) {constX=2}, B=@n101) {constB=0.2}
  n110 = Step(Y=@n100) {constX=3}
  n111 = OneMinus(None=Add(A=Add(A=@n108, B=@n109), B=Multiply(A=@n110, B=@n101) {constB=0.2}))
  n112 = NamedRerouteDeclaration(None=Multiply(A=SmoothStep(Min=@n98, Max=@n99, Value=@n111)) {constB=2}) {name=Steped Terrain, nodeColor=<rgba 0.5703 1 0 1>}
  n113 = Constant() {r=0.1}
  NamedRerouteDeclaration(None=Sign(None=MaterialFunctionCall(In=@n112, Contrast=@n113) {materialFunction=&CheapContrast}.Result)) {name=Water, nodeColor=<rgba 0 1 0.0625 1>}
  n114 = Time()
  n115 = Sine(None=Multiply(A=Divide(A=@n114) {constB=15}, B=Constant() {r=0.8}))
  n116 = Add(A=@n115)
  NamedRerouteDeclaration(None=MaterialFunctionCall(In=Multiply(A=@n116) {constB=0.5}, Contrast=Constant() {r=2}) {materialFunction=&CheapContrast}.Result) {name=Switch, nodeColor=<rgba 1 0 0.6562 1>}
  NamedRerouteDeclaration(None=TextureSample(UVs=ComponentMask(None=@n96) {a=true, b=true}) {samplerType=SAMPLERTYPE_LinearColor, texture=&T_RandoMPixelNoise}.R) {name=Random 2, nodeColor=<rgba 1 0.0703 0 1>}
  n117 = Multiply(A=Divide(A=SceneDepth().R, B=PixelDepth().R), B=WorldPosition() {worldPositionShaderOffset=WPT_CameraRelative}.XYZ)
  n118 = ActorPositionWS() {originType=CameraRelative}
  n119 = Transform(None=Divide(A=Subtract(A=@n117, B=@n118)) {constB=125}) {transformSourceType=TRANSFORMSOURCE_World, transformType=TRANSFORM_Local}
  NamedRerouteDeclaration(None=ComponentMask(None=@n119) {g=true, r=true}) {name=Projected UV, nodeColor=<rgba 0.8281 0 1 1>}
  n120 = Multiply(A=@n30, B=NamedRerouteUsage())
  n121 = Add(A=Multiply(A=@n8.R, B=Multiply(A=@n8.B, B=@n10.G)), B=Multiply(A=@n8.R, B=@n10.B))
  n122 = Multiply(
   A=MaterialFunctionCall(Alpha Threshold=@n121) {materialFunction=&DitherTemporalAA}.Result
   B=Multiply(A=@n121) {constB=4}
  )
  n123 = NamedRerouteUsage()
  n124 = Add(A=Multiply(A=Multiply(A=@n122, B=@n123), B=ScalarParameter() {defaultValue=4, parameterName=Noise dot}) {constB=4}, B=@n120)
  n125 = MaterialFunctionCall("Xb1,PS4"=@n124, "PC-High&&Epic,XSX,PS5"=@n124, "PC-Perf,Android"=@n120, Switch=@n120, "PC-Low(NotPerf)&&Medium"=@n120) {materialFunction=&PlatformSwitch_1D}
  n126 = TextureCoordinate()
  n127 = ScalarParameter() {defaultValue=4, parameterName=Line Noise Size}
  n128 = TextureSample(UVs=Panner(Coordinate=Multiply(A=@n126, B=@n127) {constB=0.1}) {speedX=0.1}) {texture=/Game/Global/Textures/Default/Noise/3xPerlin_M}
  n129 = Constant() {r=1}
  n130 = MaterialFunctionCall("Xb1,PS4"=@n128.R, "PC-High&&Epic,XSX,PS5"=@n128.R, "PC-Perf,Android"=@n129, Switch=@n129, "PC-Low(NotPerf)&&Medium"=@n129) {materialFunction=&PlatformSwitch_1D}
  NamedRerouteDeclaration(None=Multiply(A=Multiply(A=@n125.Result, B=@n130.Result), B=ScalarParameter() {defaultValue=2, parameterName=Grid Intensity}) {constB=2}) {name=Lines, nodeColor=<rgba 1 0.3008 0.0142 1>}
  n131 = NamedRerouteUsage()
  n132 = Add(A=Abs(None=DDX(Value=@n131)), B=Abs(None=DDY(Value=@n131)))
  NamedRerouteDeclaration(None=Multiply(A=Multiply(A=@n132) {constB=0.3}, B=NamedRerouteUsage())) {name=Outlines, nodeColor=<rgba 1 0.6094 0 1>}
  n133 = Multiply(A=@n115, B=@n115)
  n134 = ScalarParameter() {defaultValue=2, parameterName=Downtime Contrast}
  n135 = MaterialFunctionCall(In=Multiply(A=Add(A=@n133) {constB=0.6}) {constB=0.9}, Contrast=@n134) {materialFunction=&CheapContrast}
  n136 = Multiply(A=@n134) {constB=5}
  NamedRerouteDeclaration(None=Saturate(None=Power(Base=@n135.Result, Exp=@n136))) {name=Downtime, nodeColor=<rgba 0.0547 1 0 1>}
  NamedRerouteDeclaration(None=ScalarParameter() {defaultValue=1, group=MonitorStatic, parameterName=StaticSpeed}) {name=StaticSpeed, nodeColor=<rgba 0 1 0.6484 1>}
  NamedRerouteDeclaration(None=@n119) {nodeColor=<rgba 1 0 0.3984 1>}
  NamedRerouteUsage()
  NamedRerouteUsage()
  NamedRerouteUsage()
  NamedRerouteUsage()
  Normalize(VectorInput=@n59)
  Normalize(VectorInput=@n4.UVs)
  n137 = NamedRerouteUsage()
  n138 = Add(A=OneMinus(None=Saturate(None=@n137)), B=ScalarParameter() {defaultValue=0.8933, parameterName=Static Min})
  n139 = NamedRerouteUsage()
  n140 = Multiply(A=NamedRerouteUsage(), B=ScalarParameter() {defaultValue=0.03, parameterName=Terrain Intensity})
  n141 = Add(A=Multiply(A=Add(A=@n139, B=@n140), B=NamedRerouteUsage()), B=NamedRerouteUsage())
  n142 = Saturate(None=Add(A=NamedRerouteUsage()) {constB=0.8})
  n143 = ScalarParameter() {bUseCustomPrimitiveData=true, parameterName=Outside Compensation}
  n144 = NamedRerouteUsage()
  n145 = Multiply(
   A=VectorParameter() {parameterName=Color Terrain, defaultValue=<rgba 0.625 0.1931 0 1>}.RGB
   B=Add(A=Multiply(A=@n141, B=@n142), B=Multiply(A=@n143, B=@n144))
  )
  n146 = ScalarParameter() {defaultValue=17, parameterName=Emissive Intensity}
  n147 = ScalarParameter() {defaultValue=30.4894, parameterName=Emissive Perf}
  n148 = MaterialFunctionCall("Xb1,PS4"=@n146, "PC-High&&Epic,XSX,PS5"=@n146, "PC-Perf,Android"=@n147, Switch=@n147, "PC-Low(NotPerf)&&Medium"=@n147) {materialFunction=&PlatformSwitch_1D}
  n149 = Multiply(A=NamedRerouteUsage()) {constB=3}
  n150 = Constant() {r=1.5}
  n151 = Multiply(
   A=Multiply(A=Saturate(None=@n138), B=Multiply(A=@n145, B=@n148.Result))
   B=LinearInterpolate(A=Add(A=@n149, B=@n150), Alpha=NamedRerouteUsage()) {constA=3}
  )
  n152 = ScalarParameter() {defaultValue=1, parameterName=EyeAdapt}
  n153 = NamedRerouteUsage()
  n154 = Multiply(A=NamedRerouteUsage(), B=Saturate(None=Add(A=@n153) {constB=0.8}))
  n155 = Constant() {r=-0.8901}
  n156 = NamedRerouteUsage()
  n157 = MaterialFunctionCall(Material_In=MakeMaterialAttributes!(
    BaseColor=Constant()
    Metallic
    Specular
    Roughness
    Anisotropy
    EmissiveColor=EyeAdaptationInverse(LightValueInput=@n151, AlphaInput=@n152).EyeAdaptationInverse
    Opacity=MaterialFunctionCall(Alpha Threshold=@n154, Random=@n155) {materialFunction=&DitherTemporalAA}.Result
    OpacityMask
    Normal
    Tangent
    WorldPositionOffset=@n156
    SubsurfaceColor
    ClearCoat
    ClearCoatRoughness
    AmbientOcclusion
    Refraction
    CustomizedUVs_0
    CustomizedUVs_1
    CustomizedUVs_2
    CustomizedUVs_3
    CustomizedUVs_4
    CustomizedUVs_5
    CustomizedUVs_6
    CustomizedUVs_7
    PixelDepthOffset
    ShadingModel
    Displacement
   )) {materialFunction=&MF_FN_DestructibleDamage}
  SetMaterialAttributes!(MaterialAttributes=@n157.Material_Out, World Position Offset=Add(A=@n157.DamageAnimWPO, B=@n156)) {attributeSetTypes [#1]}
  TextureCoordinate()
 ]
}
