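# M_UVtrap is a translucent lit material that renders a tinted, emissive, animated monitor-static effect, apparently for the UV trap.
# It builds grid, border, fresnel, panning static, sweep and distance-faded scanline layers as named reroutes, including UV distortion terms.
# A pivot-seeded time offset varies animation per object.
# Combined layers drive eye-adaptation-compensated emissive over a scalar opacity, then pass through MF_FN_DestructibleDamage.
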Material /Hermes/Techart/Crewcut/M_UVtrap
graph {
 properties {bAutomaticallySetUsageInEditor=false, bUseMaterialAttributes=true, blendMode=BLEND_Translucent, usedShadingModels=Default Lit}
 nodes [
  n1 = Multiply(A=Add(A=NamedRerouteUsage(), B=TextureCoordinate()), B=ScalarParameter() {defaultValue=21.6732, parameterName=Grid Amount})
  n2 = Multiply(A=Add(A=Sine(None=@n1))) {constB=0.5}
  n3 = Saturate(None=Power(Base=Abs(None=@n2), Exp=ScalarParameter() {defaultValue=40, parameterName=Grid Thickness}))
  Add(A=ComponentMask(None=@n3) {r=true}, B=ComponentMask(None=@n3) {g=true})
  ComponentMask(None=VectorParameter() {group=MonitorStatic, parameterName=TexCoordScale, defaultValue=<rgba 0.25 0.25 0 0>}.RGB) {g=true, r=true}
  n4 = Multiply(A=Add(A=TextureCoordinate()) {constB=-0.5}) {constB=4}
  n5 = Saturate(None=Add(A=Abs(None=@n4), B=ScalarParameter() {defaultValue=-1.6, parameterName=Border Size}) {constB=-1})
  n6 = Multiply(A=@n5, B=@n5)
  n7 = Constant()
  NamedRerouteDeclaration(None=Multiply(A=Distance(A=@n6, B=@n7), B=ScalarParameter() {defaultValue=70, parameterName=Border Intensity})) {name=Border, nodeColor=<rgba 0 0.25 1 1>}
  n8 = CameraVectorWS()
  n9 = VertexNormalWS()
  n10 = Power(Base=OneMinus(None=DotProduct(A=@n8, B=@n9))) {constExponent=3}
  n11 = ScalarParameter() {defaultValue=4.4, parameterName=Frensnel Intensity}
  NamedRerouteDeclaration(None=Add(A=Multiply(A=@n10, B=@n11) {constB=0.1}) {constB=0.65}) {name=Fresnel, nodeColor=<rgba 1 0.4195 0 1>}
  n12 = NamedRerouteUsage()
  n13 = NamedRerouteUsage()
  n14 = Constant2Vector() {g=0.1, r=5}
  n15 = NamedRerouteUsage()
  n16 = TextureSample(UVs=Panner(Coordinate=@n12, Time=@n13, Speed=Multiply(A=@n14, B=@n15)) {speedX=5, speedY=0.1}) {
   samplerSource=SSM_Wrap_WorldGroupSettings
   samplerType=SAMPLERTYPE_Masks
   texture=/Game/Missions/Primary/RepairTheShelter/Textures/Cinematic_Screen_Static
  }
  n17 = Multiply(A=@n16.B) {constB=0.1}
  n18 = Constant2Vector() {g=-0.05, r=6}
  n19 = NamedRerouteUsage()
  n20 = TextureSample(UVs=Panner(Coordinate=@n12, Time=@n13, Speed=Multiply(A=@n18, B=@n19)) {speedX=6, speedY=-0.05}) {
   samplerSource=SSM_Wrap_WorldGroupSettings
   samplerType=SAMPLERTYPE_Masks
   texture=&Cinematic_Screen_Static
  }
  n21 = Constant2Vector() {g=-0.03, r=-3}
  n22 = NamedRerouteUsage()
  n23 = TextureSample(UVs=Panner(Coordinate=@n12, Time=@n13, Speed=Multiply(A=@n21, B=@n22)) {speedX=-3, speedY=-0.03}) {
   samplerSource=SSM_Wrap_WorldGroupSettings
   samplerType=SAMPLERTYPE_Masks
   texture=&Cinematic_Screen_Static
  }
  n24 = Multiply(A=@n20.G, B=@n23.R)
  NamedRerouteDeclaration(None=Multiply(A=Add(A=@n17, B=@n24), B=ScalarParameter() {defaultValue=13.4206, group=MonitorStatic, parameterName=Static Scale})) {name=Static, nodeColor=<rgba 0 1 0.8828 1>}
  n25 = Multiply(A=@n23.G, B=@n16.R)
  NamedRerouteDeclaration(None=Multiply(A=AppendVector(A=@n24, B=@n25), B=ScalarParameter() {defaultValue=0.2, group=MonitorStatic, parameterName=Static UV Distort})) {name=StaticUVDistrotion, nodeColor=<rgba 1 0.5625 0 1>}
  n26 = TextureCoordinate()
  n27 = ScalarParameter() {defaultValue=0.25, parameterName=Noise Size}
  n28 = VectorParameter() {group=MonitorStatic, parameterName=TexCoordOffset}
  NamedRerouteDeclaration(None=Add(A=Multiply(A=@n26, B=@n27), B=ComponentMask(None=@n28.RGB) {g=true, r=true})) {name=TexCoord, nodeColor=<rgba 1 0 0.4453 1>}
  n29 = Multiply(A=ComponentMask(None=NamedRerouteUsage()) {g=true}) {constB=2}
  n30 = Multiply(A=NamedRerouteUsage(), B=ScalarParameter() {defaultValue=-0.2, parameterName=Sweep Time}) {constB=-0.2}
  n31 = Power(Base=Frac(None=Add(A=@n29, B=@n30))) {constExponent=20}
  NamedRerouteDeclaration(None=Multiply(A=@n31, B=ScalarParameter() {defaultValue=3, group=MonitorStatic, parameterName=Sweep Add}) {constB=0.2}) {name=Sweep, nodeColor=<rgba 0.2422 0 1 1>}
  n32 = ScalarParameter() {defaultValue=0.01, group=MonitorStatic, parameterName=Sweep UV Distort}
  NamedRerouteDeclaration(None=AppendVector(A=Multiply(A=@n31, B=@n32) {constB=0.02}, B=Constant())) {name=SweepUVDistortion, nodeColor=<rgba 0.0781 1 0 1>}
  n33 = Add(A=@n16.RGB, B=@n20.RGB)
  NamedRerouteDeclaration(None=Add(A=Multiply(A=@n33, B=@n23.RGB), B=@n23.RGB)) {name=Static Color, nodeColor=<rgba 0.3359 1 0 1>}
  NamedRerouteDeclaration() {name=Grid, nodeColor=<rgba 0.0167 0.9235 1 1>}
  n34 = MaterialFunctionCall() {materialFunction=&ObjectPivotPoint}
  n35 = MaterialFunctionCall(Float3=Frac(None=Multiply(A=@n34.Object Pivot Location) {constB=0.15})) {materialFunction=&BreakOutFloat3Components}
  n36 = Add(A=@n35.R, B=@n35.R)
  NamedRerouteDeclaration(None=Add(A=Add(A=@n36, B=@n35.B), B=Time())) {name=Random TIme, nodeColor=<rgba 0.5 1 0 1>}
  NamedRerouteDeclaration(None=ScalarParameter() {defaultValue=1, group=MonitorStatic, parameterName=StaticSpeed}) {name=StaticSpeed, nodeColor=<rgba 0 1 0.6484 1>}
  n37 = TextureCoordinate()
  n38 = ConstantBiasScale(None=Frac(None=Panner(Coordinate=@n37) {speedY=-1})) {bias=-0.5, scale=2}
  n39 = ComponentMask(None=Multiply(A=@n38, B=@n38)) {g=true}
  n40 = ScalarParameter() {defaultValue=1, group=MonitorStatic, parameterName=Scanline Strength}
  NamedRerouteDeclaration(None=Add(A=Multiply(A=@n39, B=@n40) {constB=0.1}) {constB=0.8}) {name=Scanlines, nodeColor=<rgba 0 0.1562 1 1>}
  n41 = ScalarParameter() {defaultValue=0.8}
  n42 = ScalarParameter() {defaultValue=30, parameterName=Param_1}
  n43 = NamedRerouteUsage()
  n44 = NamedRerouteUsage()
  n45 = Add(A=Multiply(A=Add(A=@n43, B=@n44)), B=TextureCoordinate())
  n46 = ScalarParameter() {defaultValue=50, group=MonitorStatic, parameterName=ScanlineDistanceFadeRange}
  n47 = Multiply(A=ScalarParameter() {defaultValue=497.8171, group=MonitorStatic, parameterName=ScanlineDistanceFadeStart}) {constB=-1}
  NamedRerouteDeclaration(None=LinearInterpolate(
    A=MaterialFunctionCall!(ScanLine_Tiling_1=@n41, Scanline_PannerSpeed_1, ScanLine_Tiling=@n42, Scanline_PannerSpeed, UVs=@n45) {materialFunction=/Game/Weapons/WeaponSkins/MaterialFunction/ScanLines}.Result
    B=Constant() {r=0.5}
    Alpha=MaterialFunctionCall!(Blend Range=@n46, Start Offset=@n47) {materialFunction=/Engine/Functions/Engine_MaterialFunctions03/Particles/Distance_Blend}.Result
   )) {name=Scanlines_02, nodeColor=<rgba 0.4766 1 0 1>}
  NamedRerouteUsage()
  n48 = Multiply(A=Add(A=NamedRerouteUsage(), B=NamedRerouteUsage()), B=NamedRerouteUsage())
  n49 = Saturate(None=Add(A=NamedRerouteUsage(), B=ScalarParameter() {defaultValue=0.5131, parameterName=Static Add}) {constB=0.85})
  n50 = Multiply(A=Add(A=NamedRerouteUsage(), B=Multiply(A=@n48, B=@n49)), B=Add(A=NamedRerouteUsage()) {constB=0.5})
  n51 = ScalarParameter() {defaultValue=0.5, parameterName=Add Opacity}
  n52 = Multiply(
   A=ScalarParameter() {defaultValue=5, parameterName=Emissive Intensity}
   B=Multiply(A=VectorParameter() {parameterName=Color, defaultValue=<rgba 0 0.024 0.0573 1>}.RGB, B=Add(A=@n50, B=@n51))
  )
  n53 = Constant() {r=0.5}
  n54 = MaterialFunctionCall(Material_In=MakeMaterialAttributes!(
    BaseColor=Constant()
    Metallic
    Specular
    Roughness
    Anisotropy
    EmissiveColor=EyeAdaptationInverse(LightValueInput=@n52, AlphaInput=@n53).EyeAdaptationInverse
    Opacity=ScalarParameter() {defaultValue=0.8, parameterName=Opacity}
    OpacityMask
    Normal
    Tangent
    WorldPositionOffset
    SubsurfaceColor
    ClearCoat
    ClearCoatRoughness
    AmbientOcclusion
    Refraction
    CustomizedUVs_0
    CustomizedUVs_1
    CustomizedUVs_2
    CustomizedUVs_3
    CustomizedUVs_4
    CustomizedUVs_5
    CustomizedUVs_6
    CustomizedUVs_7
    PixelDepthOffset
    ShadingModel
    Displacement
   )) {materialFunction=&MF_FN_DestructibleDamage}
  SetMaterialAttributes!(MaterialAttributes=@n54.Material_Out, World Position Offset=@n54.Material_Out) {attributeSetTypes [#1]}
  n55 = Constant()
  SingleLayerWaterMaterialOutput!(ScatteringCoefficients=@n55, AbsorptionCoefficients=@n55, PhaseG=@n55, ColorScaleBehindWater=@n55)
  Time()
 ]
}
